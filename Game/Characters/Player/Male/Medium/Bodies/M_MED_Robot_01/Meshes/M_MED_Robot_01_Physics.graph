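# Physics asset for the M_MED_Robot_01 male medium player body, covering only its dyn_ accessory bones rather than a full ragdoll.
# Small capsules on the pelvis and forearms anchor capsules for two elbow pads and a rear pipe.
# Each accessory has a hinge-like constraint to its anchor, swinging mainly on one axis, with the pipe tighter than the pads.
# Forearms link to the pelvis with free swing and twist.
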PhysicsAsset /Game/Characters/Player/Male/Medium/Bodies/M_MED_Robot_01/Meshes/M_MED_Robot_01_Physics
bodies {
 dyn_butt_pipe [{shapeType=Capsule, length=10.25, radius=3.5, center=<-1.1475 -8.9889 0.0001>}]
 dyn_elbowpad_l [{shapeType=Capsule, length=2, radius=2, center=<-1.9609 -0.9478 0>}]
 dyn_elbowpad_r [{shapeType=Capsule, length=2, radius=2, center=<1.9609 0.9478 0>}]
 lowerarm_l [{shapeType=Capsule, length=2, radius=2}]
 lowerarm_r [{shapeType=Capsule, length=2, radius=2}]
 pelvis [{shapeType=Capsule, length=2, radius=2}]
}
constraints [
 {
  bone1Name=lowerarm_l
  bone2Name=pelvis
  swing1LimitDegrees=45
  swing1Motion=Free
  swing2LimitDegrees=45
  swing2Motion=Free
  twistLimitDegrees=45
  twistMotion=Free
 }
 {bone1Name=dyn_elbowpad_l, bone2Name=lowerarm_l, swing1LimitDegrees=45, swing2LimitDegrees=1, twistLimitDegrees=1}
 {
  bone1Name=lowerarm_r
  bone2Name=pelvis
  swing1LimitDegrees=45
  swing1Motion=Free
  swing2LimitDegrees=45
  swing2Motion=Free
  twistLimitDegrees=45
  twistMotion=Free
 }
 {bone1Name=dyn_elbowpad_r, bone2Name=lowerarm_r, swing1LimitDegrees=45, swing2LimitDegrees=1, twistLimitDegrees=1}
 {bone1Name=dyn_butt_pipe, bone2Name=pelvis, swing1LimitDegrees=10, swing2LimitDegrees=1, twistLimitDegrees=1}
]
