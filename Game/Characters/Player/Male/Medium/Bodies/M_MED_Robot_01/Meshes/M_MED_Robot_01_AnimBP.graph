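# Secondary-motion AnimBP for the Robot_01 medium male player body, copying the pose from its attached parent mesh.
# One chain aims piston pairs with Look Ats, constrains kneepads, then runs RigidBody and elbow-pad AnimDynamics.
# Weighted constraints, spine_03-driven offsets and chained Look Ats shape the back spine, before scapula constraints finish the pose.
# Sine-based jitter is inactive: its Transform Bone nodes never reach output, and the update event is unwired.
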AnimBlueprint /Game/Characters/Player/Male/Medium/Bodies/M_MED_Robot_01/Meshes/M_MED_Robot_01_AnimBP
target_skeleton=&/M_MED_Robot_01_Skeleton
graphs [
 AnimGraph {nodes [
   n1 = Get Random Rotation 02() -> (Random Rotation 02:Rotator)
   n2 = "Transform (Modify) Bone - Bone: scapula_armor_l"!(ComponentPose, Alpha=1, Rotation=@n1) -> (Pose) {
    rotationMode=BMM_Additive
    translationSpace=BCS_ParentBoneSpace
    alphaBoolBlend {blendOption=Linear}
    alphaScaleBias {scale=0.1667}
    alphaScaleBiasClamp {bInterpResult=true, interpSpeedDecreasing=0.1, interpSpeedIncreasing=0.1}
    boneToModify {boneName=scapula_armor_l}
   }
   n3 = Get Random_Rotation_01() -> (Random_Rotation_01:Rotator)
   n4 = "Transform (Modify) Bone - Bone: scapula_armor_r"!(ComponentPose=@n2, Alpha=1, Rotation=@n3) -> (Pose) {
    rotationMode=BMM_Additive
    translationSpace=BCS_ParentBoneSpace
    alphaBoolBlend {blendOption=Linear}
    alphaScaleBias {scale=0.25}
    alphaScaleBiasClamp {bInterpResult=true, interpSpeedDecreasing=0.1, interpSpeedIncreasing=0.1}
    boneToModify {boneName=scapula_armor_r}
   }
   "Transform (Modify) Bone - Bone: scapula_cylinder"!(ComponentPose=@n4, Alpha=1, Rotation=@n1) -> (Pose) {
    rotationMode=BMM_Additive
    translationSpace=BCS_ParentBoneSpace
    alphaBoolBlend {blendOption=Linear}
    alphaScaleBias {scale=0.2857}
    alphaScaleBiasClamp {bInterpResult=true, interpSpeedDecreasing=0.5, interpSpeedIncreasing=0.5}
    boneToModify {boneName=scapula_cylinder}
   }
   "Transform (Modify) Bone - Bone: dyn_butt_pipe"!(ComponentPose, Alpha=1, Rotation=@n3) -> (Pose) {
    rotationMode=BMM_Additive
    translationSpace=BCS_ParentBoneSpace
    alphaBoolBlend {blendOption=Linear}
    alphaScaleBias {scale=0.2}
    alphaScaleBiasClamp {bInterpResult=true, interpSpeedDecreasing=0.1, interpSpeedIncreasing=0.1}
    boneToModify {boneName=dyn_butt_pipe}
   }
   n5 = Look At - Bone: dyn_belly_piston_top_l(ComponentPose=Local To Component(LocalPose=Copy Pose From Mesh() {bUseAttachedParent=true}), Alpha=1) {
    alphaBoolBlend {blendOption=Linear}
    boneToModify {boneName=dyn_belly_piston_top_l}
    lookAtLocation {x=0}
    lookAtTarget {boneReference {boneName=dyn_belly_piston_bott_l}}
    lookAt_Axis {axis=<-1 0>}
   }
   n6 = Look At - Bone: dyn_belly_piston_bott_l(ComponentPose=@n5, Alpha=1) {
    alphaBoolBlend {blendOption=Linear}
    boneToModify {boneName=dyn_belly_piston_bott_l}
    lookAtLocation {x=0}
    lookAtTarget {boneReference {boneName=dyn_belly_piston_top_l}}
    lookAt_Axis {axis=<1 0>}
   }
   n7 = Look At - Bone: dyn_belly_piston_top_r(ComponentPose=@n6, Alpha=1) {
    alphaBoolBlend {blendOption=Linear}
    boneToModify {boneName=dyn_belly_piston_top_r}
    lookAtLocation {x=0}
    lookAtTarget {boneReference {boneName=dyn_belly_piston_bott_r}}
    lookAt_Axis {axis=<1 0>}
   }
   n8 = Look At - Bone: dyn_belly_piston_bott_r(ComponentPose=@n7, Alpha=1) {
    alphaBoolBlend {blendOption=Linear}
    boneToModify {boneName=dyn_belly_piston_bott_r}
    lookAtLocation {x=0}
    lookAtTarget {boneReference {boneName=dyn_belly_piston_top_r}}
    lookAt_Axis {axis=<-1 0>}
   }
   n9 = Look At - Bone: dyn_tricep_piston_top_l(ComponentPose=@n8, Alpha=1) {
    alphaBoolBlend {blendOption=Linear}
    boneToModify {boneName=dyn_tricep_piston_top_l}
    lookAtLocation {x=0}
    lookAtTarget {boneReference {boneName=dyn_tricep_piston_bott_l}}
    lookAt_Axis {axis=<-1 0>}
   }
   n10 = Look At - Bone: dyn_tricep_piston_bott_l(ComponentPose=@n9, Alpha=1) {
    alphaBoolBlend {blendOption=Linear}
    boneToModify {boneName=dyn_tricep_piston_bott_l}
    lookAtLocation {x=0}
    lookAtTarget {boneReference {boneName=dyn_tricep_piston_top_l}}
    lookAt_Axis {axis=<1 0>}
   }
   n11 = Look At - Bone: dyn_tricep_piston_top_r(ComponentPose=@n10, Alpha=1) {
    alphaBoolBlend {blendOption=Linear}
    boneToModify {boneName=dyn_tricep_piston_top_r}
    lookAtLocation {x=0}
    lookAtTarget {boneReference {boneName=dyn_tricep_piston_bott_r}}
    lookAt_Axis {axis=<1 0>}
   }
   n12 = Look At - Bone: dyn_tricep_piston_bott_r(ComponentPose=@n11, Alpha=1) {
    alphaBoolBlend {blendOption=Linear}
    boneToModify {boneName=dyn_tricep_piston_bott_r}
    lookAtLocation {x=0}
    lookAtTarget {boneReference {boneName=dyn_tricep_piston_top_r}}
    lookAt_Axis {axis=<-1 0>}
   }
   n13 = Look At - Bone: dyn_shin_piston_top_l(ComponentPose=@n12, Alpha=1) {
    alphaBoolBlend {blendOption=Linear}
    boneToModify {boneName=dyn_shin_piston_top_l}
    lookAtLocation {x=0}
    lookAtTarget {boneReference {boneName=dyn_shin_piston_bott_l}}
    lookAt_Axis {axis=<-1 0>}
   }
   n14 = Look At - Bone: dyn_shin_piston_bott_l(ComponentPose=@n13, Alpha=1) {
    alphaBoolBlend {blendOption=Linear}
    boneToModify {boneName=dyn_shin_piston_bott_l}
    lookAtLocation {x=0}
    lookAtTarget {boneReference {boneName=dyn_shin_piston_top_l}}
    lookAt_Axis {axis=<1 0>}
   }
   n15 = Look At - Bone: dyn_shin_piston_top_r(ComponentPose=@n14, Alpha=1) {
    alphaBoolBlend {blendOption=Linear}
    boneToModify {boneName=dyn_shin_piston_top_r}
    lookAtLocation {x=0}
    lookAtTarget {boneReference {boneName=dyn_shin_piston_bott_r}}
    lookAt_Axis {axis=<1 0>}
   }
   n16 = Look At - Bone: dyn_shin_piston_bott_r(ComponentPose=@n15, Alpha=1) {
    alphaBoolBlend {blendOption=Linear}
    boneToModify {boneName=dyn_shin_piston_bott_r}
    lookAtLocation {x=0}
    lookAtTarget {boneReference {boneName=dyn_shin_piston_top_r}}
    lookAt_Axis {axis=<-1 0>}
   }
   n17 = Look At - Bone: dyn_achilles_piston_top_l(ComponentPose=@n16, Alpha=1) {
    alphaBoolBlend {blendOption=Linear}
    boneToModify {boneName=dyn_achilles_piston_top_l}
    lookAtLocation {x=0}
    lookAtTarget {boneReference {boneName=dyn_achilles_piston_bott_l}}
    lookAt_Axis {axis=<-1 0>}
   }
   n18 = Look At - Bone: dyn_achilles_piston_bott_l(ComponentPose=@n17, Alpha=1) {
    alphaBoolBlend {blendOption=Linear}
    boneToModify {boneName=dyn_achilles_piston_bott_l}
    lookAtLocation {x=0}
    lookAtTarget {boneReference {boneName=dyn_achilles_piston_top_l}}
    lookAt_Axis {axis=<1 0>}
   }
   n19 = Look At - Bone: dyn_achilles_piston_top_r(ComponentPose=@n18, Alpha=1) {
    alphaBoolBlend {blendOption=Linear}
    boneToModify {boneName=dyn_achilles_piston_top_r}
    lookAtLocation {x=0}
    lookAtTarget {boneReference {boneName=dyn_achilles_piston_bott_r}}
    lookAt_Axis {axis=<1 0>}
   }
   n20 = Look At - Bone: dyn_achilles_piston_bott_r(ComponentPose=@n19, Alpha=1) {
    alphaBoolBlend {blendOption=Linear}
    boneToModify {boneName=dyn_achilles_piston_bott_r}
    lookAtLocation {x=0}
    lookAtTarget {boneReference {boneName=dyn_achilles_piston_top_r}}
    lookAt_Axis {axis=<-1 0>}
   }
   n21 = "Constraint - kneepad_l (PP)"!(ComponentPose=@n20, Alpha=1, ConstraintWeights_0=0.3832, ConstraintWeights_1:float=0.6168) -> (Pose) {
    alphaBoolBlend {blendOption=Linear}
    boneToModify {boneName=kneepad_l}
    constraintSetup [{targetBone {boneName=calf_l}}, {targetBone {boneName=thigh_l}}]
    constraintWeights [1,1]
   }
   n22 = "Constraint - kneepad_r (PP)"!(ComponentPose=@n21, Alpha=1, ConstraintWeights_0=0.3832, ConstraintWeights_1:float=0.6168) -> (Pose) {
    alphaBoolBlend {blendOption=Linear}
    boneToModify {boneName=kneepad_r}
    constraintSetup [{targetBone {boneName=calf_r}}, {targetBone {boneName=thigh_r}}]
    constraintWeights [1,1]
   }
   n23 = RigidBody(ComponentPose=@n22) {
    bOverrideWorldGravity=true
    overridePhysicsAsset=&/M_MED_Robot_01_Physics
    simulationSpace=BaseBoneSpace
    alphaBoolBlend {blendOption=Linear}
    baseBoneRef {boneName=pelvis}
    componentLinearAccScale=<2 2 2>
    componentLinearVelScale=<2 2 2>
    overrideWorldGravity {z=-980}
   }
   n24 = Anim Dynamics - Bone: dyn_elbowpad_l(ComponentPose=@n23, Alpha=1, ExternalForce="0,0,0") {
    angularDampingOverride=0.6
    angularSpringConstant=40
    bOverrideAngularDamping=true
    bUseGravityOverride=true
    simulationSpace=BoneRelative
    alphaBoolBlend {blendOption=Linear}
    boundBone {boneName=dyn_elbowpad_l}
    componentAppliedLinearAccClamp=<10000 10000 10000>
    componentLinearAccScale=<2 2 2>
    componentLinearVelScale=<2 2 2>
    gravityOverride {z=-400}
    physicsBodyDefinitions [{
      boundBone {boneName=dyn_elbowpad_l}
      boxExtents=<5 5 5>
      constraintSetup {angularLimitsMax=<0.5 0.5 25>, angularLimitsMin=<-0.5 -0.5 -25>, angularTarget=<1 0 0>}
      localJointOffset=<0 0.02 0>
     }]
    relativeSpaceBone {boneName=lowerarm_l}
   }
   n25 = Anim Dynamics - Bone: dyn_elbowpad_r(ComponentPose=@n24, Alpha=1, ExternalForce="0,0,0") {
    angularDampingOverride=0.6
    angularSpringConstant=40
    bOverrideAngularDamping=true
    bUseGravityOverride=true
    simulationSpace=BoneRelative
    alphaBoolBlend {blendOption=Linear}
    boundBone {boneName=dyn_elbowpad_r}
    componentAppliedLinearAccClamp=<10000 10000 10000>
    componentLinearAccScale=<2 2 2>
    componentLinearVelScale=<2 2 2>
    gravityOverride {z=-400}
    physicsBodyDefinitions [{
      boundBone {boneName=dyn_elbowpad_r}
      boxExtents=<5 5 5>
      constraintSetup {angularLimitsMax=<0.5 0.5 25>, angularLimitsMin=<-0.5 -0.5 -25>, angularTarget=<1 0 0>}
      localJointOffset=<0 0.02 0>
     }]
    relativeSpaceBone {boneName=lowerarm_r}
   }
   n26 = "Constraint - dyn_back_spine_06 (PPPP)"!(
    ComponentPose=@n25
    Alpha=1
    ConstraintWeights_0=0.0084
    ConstraintWeights_1:float=0.0502
    ConstraintWeights_2:float=0.9344
    ConstraintWeights_3:float=0.007
   ) -> (Pose) {
    alphaBoolBlend {blendOption=Linear}
    boneToModify {boneName=dyn_back_spine_06}
    constraintSetup [{targetBone {boneName=pelvis}}, {targetBone {boneName=spine_02}}, {targetBone {boneName=spine_03}}, {targetBone {boneName=spine_01}}]
    constraintWeights [1,1,1,1]
   }
   n27 = "Constraint - dyn_back_spine_05 (PPPP)"!(
    ComponentPose=@n26
    Alpha=1
    ConstraintWeights_0=0.0354
    ConstraintWeights_1:float=0.1647
    ConstraintWeights_2:float=0.7059
    ConstraintWeights_3:float=0.0941
   ) -> (Pose) {
    alphaBoolBlend {blendOption=Linear}
    boneToModify {boneName=dyn_back_spine_05}
    constraintSetup [{targetBone {boneName=pelvis}}, {targetBone {boneName=spine_02}}, {targetBone {boneName=spine_03}}, {targetBone {boneName=spine_01}}]
    constraintWeights [1,1,1,1]
   }
   n28 = "Constraint - dyn_back_spine_04 (PPPP)"!(
    ComponentPose=@n27
    Alpha=1
    ConstraintWeights_0=0.1097
    ConstraintWeights_1:float=0.3352
    ConstraintWeights_2:float=0.3769
    ConstraintWeights_3:float=0.1783
   ) -> (Pose) {
    alphaBoolBlend {blendOption=Linear}
    boneToModify {boneName=dyn_back_spine_04}
    constraintSetup [{targetBone {boneName=pelvis}}, {targetBone {boneName=spine_02}}, {targetBone {boneName=spine_03}}, {targetBone {boneName=spine_01}}]
    constraintWeights [1,1,1,1]
   }
   n29 = "Constraint - dyn_back_spine_03 (PPPP)"!(
    ComponentPose=@n28
    Alpha=1
    ConstraintWeights_0=0.4798
    ConstraintWeights_1:float=0.2381
    ConstraintWeights_2:float=0.0934
    ConstraintWeights_3:float=0.1887
   ) -> (Pose) {
    alphaBoolBlend {blendOption=Linear}
    boneToModify {boneName=dyn_back_spine_03}
    constraintSetup [{targetBone {boneName=pelvis}}, {targetBone {boneName=spine_02}}, {targetBone {boneName=spine_03}}, {targetBone {boneName=spine_01}}]
    constraintWeights [1,1,1,1]
   }
   n30 = "Constraint - dyn_back_spine_02 (PP)"!(ComponentPose=@n29, Alpha=1, ConstraintWeights_0=0.9816, ConstraintWeights_1:float=0.0184) -> (Pose) {
    alphaBoolBlend {blendOption=Linear}
    boneToModify {boneName=dyn_back_spine_02}
    constraintSetup [{targetBone {boneName=pelvis}}, {targetBone {boneName=spine_01}}]
    constraintWeights [1,1]
   }
   n31 = "dyn_back_spine_05.translateX = remap(spine_03.rotateZ) * 0.05 - Bone Driven Controller"!(ComponentPose:Component Space Pose Link Structure=@n30, Alpha:float=1) -> (Pose:Component Space Pose Link Structure) {
    is=title AnimGraphNode_BoneDrivenController
    bAffectTargetTranslationX=true
    bUseRange=true
    multiplier=0.05
    rangeMax=30
    rangeMin=5
    remappedMax=30
    remappedMin=5
    sourceComponent=RotationZ
    alphaBoolBlend {blendOption=Linear}
    sourceBone {boneName=spine_03}
    targetBone {boneName=dyn_back_spine_05}
   }
   n32 = "dyn_back_spine_04.translateX = remap(spine_03.rotateZ) * 0.13 - Bone Driven Controller"!(ComponentPose:Component Space Pose Link Structure=@n31, Alpha:float=1) -> (Pose:Component Space Pose Link Structure) {
    is=title AnimGraphNode_BoneDrivenController
    bAffectTargetTranslationX=true
    bUseRange=true
    multiplier=0.13
    rangeMax=30
    rangeMin=5
    remappedMax=30
    remappedMin=5
    sourceComponent=RotationZ
    alphaBoolBlend {blendOption=Linear}
    sourceBone {boneName=spine_03}
    targetBone {boneName=dyn_back_spine_04}
   }
   n33 = "dyn_back_spine_03.translateX = remap(spine_03.rotateZ) * 0.21 - Bone Driven Controller"!(ComponentPose:Component Space Pose Link Structure=@n32, Alpha:float=1) -> (Pose:Component Space Pose Link Structure) {
    is=title AnimGraphNode_BoneDrivenController
    bAffectTargetTranslationX=true
    bUseRange=true
    multiplier=0.21
    rangeMax=30
    rangeMin=5
    remappedMax=30
    remappedMin=5
    sourceComponent=RotationZ
    alphaBoolBlend {blendOption=Linear}
    sourceBone {boneName=spine_03}
    targetBone {boneName=dyn_back_spine_03}
   }
   n34 = "dyn_back_spine_02.translateX = remap(spine_03.rotateZ) * 0.31 - Bone Driven Controller"!(ComponentPose:Component Space Pose Link Structure=@n33, Alpha:float=1) -> (Pose:Component Space Pose Link Structure) {
    is=title AnimGraphNode_BoneDrivenController
    bAffectTargetTranslationX=true
    bUseRange=true
    multiplier=0.31
    rangeMax=30
    rangeMin=5
    remappedMax=30
    remappedMin=5
    sourceComponent=RotationZ
    alphaBoolBlend {blendOption=Linear}
    sourceBone {boneName=spine_03}
    targetBone {boneName=dyn_back_spine_02}
   }
   n35 = "dyn_back_spine_01.translateX = remap(spine_03.rotateZ) * 0.32 - Bone Driven Controller"!(ComponentPose:Component Space Pose Link Structure=@n34, Alpha:float=1) -> (Pose:Component Space Pose Link Structure) {
    is=title AnimGraphNode_BoneDrivenController
    bAffectTargetTranslationX=true
    bUseRange=true
    multiplier=0.32
    rangeMax=30
    rangeMin=5
    remappedMax=30
    remappedMin=5
    sourceComponent=RotationZ
    alphaBoolBlend {blendOption=Linear}
    sourceBone {boneName=spine_03}
    targetBone {boneName=dyn_back_spine_01}
   }
   n36 = Look At - Bone: dyn_back_spine_01(ComponentPose=@n35, Alpha=1) {
    alphaBoolBlend {blendOption=Linear}
    boneToModify {boneName=dyn_back_spine_01}
    lookAtLocation {x=0}
    lookAtTarget {boneReference {boneName=dyn_back_spine_02}}
    lookAt_Axis {axis=<1 0>}
   }
   n37 = Look At - Bone: dyn_back_spine_02(ComponentPose=@n36, Alpha=1) {
    alphaBoolBlend {blendOption=Linear}
    boneToModify {boneName=dyn_back_spine_02}
    lookAtLocation {x=0}
    lookAtTarget {boneReference {boneName=dyn_back_spine_03}}
    lookAt_Axis {axis=<1 0>}
   }
   n38 = Look At - Bone: dyn_back_spine_03(ComponentPose=@n37, Alpha=1) {
    alphaBoolBlend {blendOption=Linear}
    boneToModify {boneName=dyn_back_spine_03}
    lookAtLocation {x=0}
    lookAtTarget {boneReference {boneName=dyn_back_spine_04}}
    lookAt_Axis {axis=<1 0>}
   }
   n39 = Look At - Bone: dyn_back_spine_04(ComponentPose=@n38, Alpha=1) {
    alphaBoolBlend {blendOption=Linear}
    boneToModify {boneName=dyn_back_spine_04}
    lookAtLocation {x=0}
    lookAtTarget {boneReference {boneName=dyn_back_spine_05}}
    lookAt_Axis {axis=<1 0>}
   }
   n40 = Look At - Bone: dyn_back_spine_05(ComponentPose=@n39, Alpha=1) {
    alphaBoolBlend {blendOption=Linear}
    boneToModify {boneName=dyn_back_spine_05}
    lookAtLocation {x=0}
    lookAtTarget {boneReference {boneName=dyn_back_spine_06}}
    lookAt_Axis {axis=<1 0>}
   }
   n41 = Look At - Bone: dyn_back_spine_06(ComponentPose=@n40, Alpha=1) {
    alphaBoolBlend {blendOption=Linear}
    boneToModify {boneName=dyn_back_spine_06}
    lookAtLocation {x=0}
    lookAtTarget {boneReference {boneName=dyn_back_spine_07}}
    lookAt_Axis {axis=<1 0>}
   }
   n42 = "Constraint - scapula_cylinder (PP)"!(ComponentPose=@n41, Alpha=1, ConstraintWeights_0=0.8215, ConstraintWeights_1:float=0.1785) -> (Pose) {
    alphaBoolBlend {blendOption=Linear}
    boneToModify {boneName=scapula_cylinder}
    constraintSetup [{targetBone {boneName=spine_04}}, {targetBone {boneName=clavicle_l}}]
    constraintWeights [1,1]
   }
   n43 = "Constraint - scapula_armor_l (PP)"!(ComponentPose=@n42, Alpha=1, ConstraintWeights_0=0.7963, ConstraintWeights_1:float=0.2037) -> (Pose) {
    alphaBoolBlend {blendOption=Linear}
    boneToModify {boneName=scapula_armor_l}
    constraintSetup [{targetBone {boneName=spine_04}}, {targetBone {boneName=clavicle_l}}]
    constraintWeights [1,1]
   }
   n44 = "Constraint - scapula_armor_r (PP)"!(ComponentPose=@n43, Alpha=1, ConstraintWeights_0=0.2037, ConstraintWeights_1:float=0.7963) -> (Pose) {
    alphaBoolBlend {blendOption=Linear}
    boneToModify {boneName=scapula_armor_r}
    constraintSetup [{targetBone {boneName=clavicle_r}}, {targetBone {boneName=spine_04}}]
    constraintWeights [1,1]
   }
   Output Pose(Result=Component To Local(ComponentPose=@n44)) {layerGroup=None}
  ]}
 EventGraph {nodes [
   n1 = GetGameTimeInSeconds()
   n2 = "Sin (Radians)"(A=float * float(A=float * float(A=1, B=333), B=@n1))
   n3 = float * float(A=1, B=280)
   n4 = float * float(A="Sin (Radians)"(A=float * float(A=@n3, B=@n1)), B=5)
   Set Random_Rotation_01(Random_Rotation_01:Rotator=MakeRotator(Roll=float * float(A=@n2, B=10), Pitch=@n4)) -> (Output_Get:Rotator)
   n5 = "Sin (Radians)"(A=float * float(A=float * float(A=1, B=400), B=@n1))
   > Set Random Rotation 02(Random Rotation 02:Rotator=MakeRotator(Roll=float * float(A=@n5, B=7), Pitch=@n4)) -> (Output_Get:Rotator)
   Event BlueprintUpdateAnimation()
  ]}
]
variables [
 Random Rotation 02 {type=Rotator}
 Random_Rotation_01 {type=Rotator}
]
